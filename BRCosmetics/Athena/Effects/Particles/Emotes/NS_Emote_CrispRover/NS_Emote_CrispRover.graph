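# Niagara effect for the CrispRover emote: a burning pinkish-red flare with flame, light, sparks and smoke.
# Flames layers flickering sprites, mesh and light with an ignition surge, alongside one-shot spark bursts and continuous GPU sparks.
# System variables for flare duration, speed and wave-animation intensity drive size, brightness, spawn rates and smoke turbulence.
# A curl-noise ribbon smoke trail appears around nine to eleven seconds, as the flame mesh shrinks.
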NiagaraSystem /BRCosmetics/Athena/Effects/Particles/Emotes/NS_Emote_CrispRover/NS_Emote_CrispRover
dependencies {
 usedDataInterfaces [
  &NiagaraDataInterfaceVectorCurve
  &NiagaraDataInterfaceCurve
  &NiagaraDataInterfaceVector2DCurve
  &NiagaraDataInterfaceColorCurve
  &NiagaraDataInterfaceSpriteRendererInfo
 ]
 usedDynamicInputs [&VectorLength]
 usedModules [
  &SystemState
  &self:SystemSpawnScript.SystemScriptSource.SystemScriptGraph.NiagaraNodeAssignment_0.SetVariables_#1
  &self:SystemSpawnScript.SystemScriptSource.SystemScriptGraph.NiagaraNodeAssignment_1.SetVariables_#2
  &self:SystemSpawnScript.SystemScriptSource.SystemScriptGraph.NiagaraNodeAssignment_2.SetVariables_#3
  &self:SystemSpawnScript.SystemScriptSource.SystemScriptGraph.NiagaraNodeAssignment_5.SetVariables_#4
  &InterpolateOverTime
  &self:SystemSpawnScript.SystemScriptSource.SystemScriptGraph.NiagaraNodeAssignment_3.SetVariables_#5
 ]
}
emitters {
 Empty001 {
  bEnabled=false
  emitterUpdateScript [&EmitterState()]
  particleSpawnScript [&InitializeParticle(Color Mode=::NewEnumerator0, Lifetime=5, UsePositionOffset=false, Write Material Random=true)]
  particleUpdateScript [&ParticleState()]
  rendererClasses [&NiagaraSpriteRendererProperties]
  renderers [{rendererClass=&NiagaraSpriteRendererProperties}]
 }
 Flames {
  emitterUpdateScript [
   &EmitterState(Inactive Response=::NewEnumerator1, Life Cycle Mode=::NewEnumerator1, Loop Duration=1, UseLoopDelay=false)
   &SpawnBurst_Instantaneous(Spawn Probability=HlslExpression() {hlslExpression="Emitter.LoopCount == 0"}, Use Spawn Probability=true, UseLoopCountLimit=false)
  ]
  particleSpawnScript [
   &InitializeParticle(
    Color=<rgba 4 0.0863 0.02 1>
    Lifetime=&RandomRangeFloat(Minimum=1.5, Maximum=2)
    Mesh Scale Mode=::NewEnumerator3
    Mesh Uniform Scale=0.3
    Position Offset=<0 0 10>
    Sprite Size Mode=::NewEnumerator3
    Uniform Sprite Size=70
    UsePositionOffset=true
    Write Material Random=true
   )
   /SetVariables() {isSetParameters=true}
  ]
  particleUpdateScript [
   &ParticleState(
    Kill Particles When Lifetime Has Elapsed=false
    Let Infinitely Lived Particles Die When Emitter Deactivates=true
    Loop Particles Lifetime=true
   )
   &ScaleSpriteSize(
    Uniform Curve Scale=$System.FlareFlameDuration
    Uniform Curve Sprite Scale=DataInterface() {propertyValues {Curve {keys [
        <at 0 1 RCIM_Cubic>
        {arriveTangent=-0.6622, interpMode=RCIM_Cubic, leaveTangent=-0.6622, tangentMode=RCTM_User, time=0.1, value=0.775}
        {arriveTangent=-0.0208, interpMode=RCIM_Cubic, leaveTangent=-0.0208, tangentMode=RCTM_User, time=0.2, value=0.85}
        {arriveTangent=0.3451, interpMode=RCIM_Cubic, leaveTangent=0.3451, tangentMode=RCTM_User, time=0.3, value=0.875}
        {arriveTangent=-0.0022, interpMode=RCIM_Cubic, leaveTangent=-0.0022, tangentMode=RCTM_User, time=0.4, value=0.95}
        {arriveTangent=-0.0096, interpMode=RCIM_Cubic, leaveTangent=-0.0096, tangentMode=RCTM_User, time=0.5, value=0.875}
        {arriveTangent=-0.037, interpMode=RCIM_Cubic, leaveTangent=-0.037, tangentMode=RCTM_User, time=0.6, value=0.9}
        {arriveTangent=-0.0446, interpMode=RCIM_Cubic, leaveTangent=-0.0446, tangentMode=RCTM_User, time=0.7, value=0.75}
        {arriveTangent=-0.0333, interpMode=RCIM_Cubic, leaveTangent=-0.0333, tangentMode=RCTM_User, time=0.8, value=0.95}
        {arriveTangent=-0.0145, interpMode=RCIM_Cubic, leaveTangent=-0.0145, tangentMode=RCTM_User, time=0.9, value=0.875}
        <at 1 1 RCIM_Cubic>
       ]}}}
   )
   &ScaleSpriteSize(
    Uniform Curve Index=$Emitter.Age
    Uniform Curve Sprite Scale=DataInterface() {propertyValues {Curve {keys [
        {arriveTangent=5.0894, interpMode=RCIM_Cubic, leaveTangent=5.0894, tangentMode=RCTM_User, time=0, value=0.4}
        {arriveTangent=-0.2694, interpMode=RCIM_Cubic, leaveTangent=-0.2694, tangentMode=RCTM_User, time=0.15, value=1.5}
        <at 0.3 1 RCIM_Cubic>
       ]}}}
   ) {module=ScaleSpriteSize001}
   &ScaleMeshSize(Scale Factor=&MakeVector(
     X=$System.FlareFlameDuration
     Y=$System.FlareFlameDuration
     Z=&FloatFromCurve(
      FloatCurve=DataInterface() {propertyValues {Curve {keys [
          {arriveTangent=5.0894, interpMode=RCIM_Cubic, leaveTangent=5.0894, tangentMode=RCTM_User, time=0, value=0.4}
          {arriveTangent=0.2071, interpMode=RCIM_Cubic, leaveTangent=0.2071, tangentMode=RCTM_User, time=0.2, value=1.8}
          {arriveTangent=-0.0762, interpMode=RCIM_Cubic, leaveTangent=-0.0762, time=0.5, value=1}
          {arriveTangent=-0.0727, interpMode=RCIM_Cubic, leaveTangent=-0.0727, time=10.7, value=1}
          <at 11.5 0.2 RCIM_Cubic>
         ]}}}
      CurveIndex=$Emitter.Age
     )
    ))
   /SetVariables(Particles.PositionFlares:NiagaraPosition=&AddVectorToPosition(Position=$Particles.Position, Vector=<0 0 -3>)) {isSetParameters=true}
   /SetVariables(Particles.ColorFlare:LinearColor=<rgba 15.8267 0.8 40 1>) {isSetParameters=true}
   /SetVariables(Particles.SpriteFlareSize:Vector2f=&Multiply_Vector2D_ByFloat(
     Float=$System.FlareDuration
     Vector2D=&MakeVector2D(
      X=&Multiply_Float(
       A=60
       B=&FloatFromCurve(
        FloatCurve=DataInterface() {propertyValues {Curve {keys [
            <at 0 1 RCIM_Cubic>
            {arriveTangent=0.0166, interpMode=RCIM_Cubic, leaveTangent=0.0166, tangentMode=RCTM_User, time=0.1, value=0.725}
            {arriveTangent=-0.0112, interpMode=RCIM_Cubic, leaveTangent=-0.0112, tangentMode=RCTM_User, time=0.2, value=0.875}
            {arriveTangent=0.0059, interpMode=RCIM_Cubic, leaveTangent=0.0059, tangentMode=RCTM_User, time=0.3, value=0.85}
            {arriveTangent=-0.0022, interpMode=RCIM_Cubic, leaveTangent=-0.0022, tangentMode=RCTM_User, time=0.4, value=0.95}
            {arriveTangent=-0.0096, interpMode=RCIM_Cubic, leaveTangent=-0.0096, tangentMode=RCTM_User, time=0.5, value=0.8125}
            {arriveTangent=-0.037, interpMode=RCIM_Cubic, leaveTangent=-0.037, tangentMode=RCTM_User, time=0.6, value=0.85}
            {arriveTangent=0.0048, interpMode=RCIM_Cubic, leaveTangent=0.0048, tangentMode=RCTM_User, time=0.7, value=0.725}
            {arriveTangent=-0.011, interpMode=RCIM_Cubic, leaveTangent=-0.011, tangentMode=RCTM_User, time=0.8, value=0.95}
            {arriveTangent=-0.0145, interpMode=RCIM_Cubic, leaveTangent=-0.0145, tangentMode=RCTM_User, time=0.9, value=0.875}
            <at 1 1 RCIM_Cubic>
           ]}}}
        Scale Curve=&FloatFromCurve(
         FloatCurve=DataInterface() {propertyValues {Curve {keys [
             {arriveTangent=5.0894, interpMode=RCIM_Cubic, leaveTangent=5.0894, tangentMode=RCTM_User, time=0, value=0.4}
             {arriveTangent=1.2, interpMode=RCIM_Cubic, leaveTangent=1.2, time=0.1, value=2}
             <at 0.5 1 RCIM_Cubic>
            ]}}}
         CurveIndex=$Emitter.Age
        )
       )
      )
      Y=15
     )
    )) {isSetParameters=true}
   /SetVariables(Particles.ColorFlame:LinearColor=&MakeLinearColorFromVectorAndFloat(
     "Vector (RGB)"=&MakeVectorFromLinearColorRGB(Color=<rgba 160 3.2 81.5998 1>)
     "Float (Alpha)"=&Multiply_Float(
      A=&FloatFromCurve(FloatCurve=DataInterface() {propertyValues {Curve {keys [
           <at 0 1 RCIM_Cubic>
           {arriveTangent=0.0166, interpMode=RCIM_Cubic, leaveTangent=0.0166, tangentMode=RCTM_User, time=0.1, value=0.6}
           {arriveTangent=-0.0112, interpMode=RCIM_Cubic, leaveTangent=-0.0112, tangentMode=RCTM_User, time=0.2, value=0.875}
           {arriveTangent=0.0059, interpMode=RCIM_Cubic, leaveTangent=0.0059, tangentMode=RCTM_User, time=0.3, value=0.7}
           {arriveTangent=-0.0022, interpMode=RCIM_Cubic, leaveTangent=-0.0022, tangentMode=RCTM_User, time=0.4, value=0.95}
           {arriveTangent=-0.0096, interpMode=RCIM_Cubic, leaveTangent=-0.0096, tangentMode=RCTM_User, time=0.5, value=0.75}
           {arriveTangent=-0.037, interpMode=RCIM_Cubic, leaveTangent=-0.037, tangentMode=RCTM_User, time=0.6, value=0.775}
           {arriveTangent=0.0048, interpMode=RCIM_Cubic, leaveTangent=0.0048, tangentMode=RCTM_User, time=0.7, value=0.65}
           {arriveTangent=-0.011, interpMode=RCIM_Cubic, leaveTangent=-0.011, tangentMode=RCTM_User, time=0.8, value=0.9}
           {arriveTangent=-0.0145, interpMode=RCIM_Cubic, leaveTangent=-0.0145, tangentMode=RCTM_User, time=0.9, value=0.825}
           <at 1 1 RCIM_Cubic>
          ]}}})
      B=&FloatFromCurve(
       FloatCurve=DataInterface() {propertyValues {Curve {keys [
           {arriveTangent=5.0894, interpMode=RCIM_Cubic, leaveTangent=5.0894, tangentMode=RCTM_User, time=0, value=0.4}
           {arriveTangent=1.2, interpMode=RCIM_Cubic, leaveTangent=1.2, time=0.2, value=10}
           <at 0.5 1 RCIM_Cubic>
          ]}}}
       CurveIndex=$Emitter.Age
       Scale Curve=$System.FlareFlameDuration
      )
     )
    )) {isSetParameters=true}
   &CameraOffset(Camera Offset Amount=10)
   /SetVariables(Particles.ColorLight:LinearColor=&MakeLinearColorFromVectorAndFloat(
     "Vector (RGB)"=&MakeVectorFromLinearColorRGB(Color=<rgba 18 0.09 0.8362 1>)
     "Float (Alpha)"=$System.FlareFlameDuration
    )) {isSetParameters=true}
   &ScaleColor(
    Scale RGB=&VectorFromFloat(Value=&FloatFromCurve(
      FloatCurve=DataInterface() {propertyValues {Curve {keys [
          {arriveTangent=5.0894, interpMode=RCIM_Cubic, leaveTangent=5.0894, tangentMode=RCTM_User, time=0, value=0.4}
          {arriveTangent=2, interpMode=RCIM_Cubic, leaveTangent=2, time=0.1, value=6}
          <at 0.3 1 RCIM_Cubic>
         ]}}}
      CurveIndex=$Emitter.Age
      Scale Curve=$System.FlareFlameDuration
     ))
    ScaleA=true
    ScaleRGB=true
   )
   &DynamicMaterialParameters(
    Index 1 Param 4=&FloatFromCurve(
     FloatCurve=DataInterface() {propertyValues {Curve {keys [
         {arriveTangent=-0.0917, interpMode=RCIM_Cubic, leaveTangent=-0.0917, tangentMode=RCTM_User, time=0, value=1}
         {arriveTangent=-0.031, interpMode=RCIM_Cubic, leaveTangent=-0.031, tangentMode=RCTM_User, time=9, value=0.5}
         <at 11 0.1 RCIM_Cubic>
        ]}}}
     CurveIndex=$Emitter.Age
     Scale Curve=1.5
    )
    Module.Param0WriteEnabled1=false
    Module.Param1WriteEnabled1=false
    Module.Param2WriteEnabled1=false
    Module.Param3WriteEnabled1=true
    Write Parameter Index 0=false
    Write Parameter Index 1=true
   )
  ]
  rendererClasses [
   &NiagaraSpriteRendererProperties
   &NiagaraLightRendererProperties
   &NiagaraMeshRendererProperties
  ]
  renderers [
   {rendererClass=&NiagaraSpriteRendererProperties}
   {rendererClass=&NiagaraSpriteRendererProperties, rendererIndex=1}
   {rendererClass=&NiagaraLightRendererProperties, rendererIndex=2}
   {rendererClass=&NiagaraMeshRendererProperties, rendererIndex=3}
   {rendererClass=&NiagaraSpriteRendererProperties, rendererIndex=4}
  ]
 }
 GPU_Sparks {
  simTarget=GPUComputeSim
  emitterUpdateScript [
   &EmitterState()
   &SpawnRate(SpawnRate=&Multiply_Float(A=60, B=$System.FlareSparksDuration), Use Spawn Probability=false)
  ]
  particleSpawnScript [
   &InitializeParticle(
    Color=<rgba 40 9.8591 0 1>
    Lifetime=&RandomRangeFloat(Minimum=0.05, Maximum=0.15)
    Mass Max=0.05
    Mass Min=0.02
    Mass Mode=::NewEnumerator3
    Mesh Scale Mode=::NewEnumerator3
    Mesh Uniform Scale=5
    Sprite Size Max=<1.5 7>
    Sprite Size Min=<1 1>
    Sprite Size Mode=::NewEnumerator2
    UsePositionOffset=false
    Write Material Random=true
   )
   /Niagara/Modules/Spawn/Location/SphereLocation(
    Coordinate Space:ENiagaraCoordinateSpace=::Simulation
    MaskSpawnGroup=false
    Module.Override Randomness=false
    Offset=&Lerp_Vector(A=<0 0 0>, B=<0 0 0>, Alpha=&MakeFloatFromInt(Int=$Particles.SpawnGroup))
    Sphere Radius=2
    Surface Only=false
   ) {enabled=false}
   &ShapeLocation(
    MaskSpawnGroup=false
    Offset=<0 0 10>
    Offset Coordinate Space=::Local
    Offset Mode=::NewEnumerator0
    Override Seed=false
    Sphere Radius=2
    UseOwnerScale=true
   )
   &AddVelocity(
    Cone Axis=<0 0 1>
    Module.UseConeFalloff=false
    Module.UseVelDistribution=false
    Override Seed=false
    Velocity Mode=::NewEnumerator2
    Velocity Speed=&Multiply_Float(A=&RandomRangeFloat(Minimum=100, Maximum=230), B=$System.FlareSparksDuration)
   )
  ]
  particleUpdateScript [
   &ParticleState()
   &LinearForce(Force=&VectorFromCurve(
     VectorCurve=DataInterface() {propertyValues {XCurve {keys [<at 0 1>,<at 0.1 0>]}, YCurve {keys [<at 0 1>,<at 0.1 0>]}, ZCurve {keys [<at 0 1>,<at 0.1 0>]}}}
     Scale Curve=&FN_FrontEndVector3Modifier(Gameplay=<0 2000 2000>, Front End=<0 0 2500>)
    )) {enabled=false}
   &CurlNoiseForce(
    Curl Noise Cone Mask Angle=90
    Curl Noise Cone Mask Falloff Angle=90
    Mask Curl Noise=true
    Noise Frequency=200
    Noise Strength=50
    Pan Noise=false
   )
   &Drag(Drag=4, Use Linear Drag=true, Use Rotational Drag=false)
   &PointForce(Module.Use Falloff=false, Module.UseLinearForce=false) {enabled=false}
   &SolveForcesAndVelocity(Clamp Velocity=false, Limit Acceleration=false)
   &ScaleSpriteSizeBySpeed(
    Max Scale Factor=<1 1>
    Min Scale Factor=<0.5 0.5>
    Sample Scale Factor By Curve=true
    Scale Factor Curve=DataInterface() {propertyValues {Curve {keys [
        {arriveTangent=2.865, interpMode=RCIM_Cubic, time=0.1247, value=-0.0006}
        {arriveTangent=4.1522, interpMode=RCIM_Cubic, leaveTangent=4.1522, tangentMode=RCTM_User, time=0.3948, value=0.4893}
        <at 0.9958 0.9974 RCIM_Cubic>
       ]}}}
    Velocity Threshold=500
   )
   &ScaleSpriteSize(Scale Factor=&Vector2DFromCurve(
     Vector2Curve=DataInterface() {propertyValues {
       XCurve {keys [
         <at 0 0 RCIM_Cubic>
         {arriveTangent=-1.6411, interpMode=RCIM_Cubic, leaveTangent=-1.6411, tangentMode=RCTM_User, time=0.2457, value=1.0046}
         <at 1 0>
        ]}
       YCurve {keys [<at 0 0 RCIM_Cubic>,<at 0.2457 1.0046 RCIM_Cubic>,<at 1 0>]}
      }}
     ScaleCurve=<1 1>
    ))
   &Color(
    Color=&ColorFromCurve(ColorCurve=DataInterface() {propertyValues {
       AlphaCurve {keys [<at 0 0>,<at 0.0353 1 RCIM_Cubic>,<at 1 0 RCIM_Cubic>]}
       BlueCurve {keys [<at 0.0553 0.7793 RCIM_Cubic>,<at 0.4938 0>,<at 1 0 RCIM_Cubic>]}
       GreenCurve {keys [<at 0.0553 0 RCIM_Cubic>,<at 0.4938 0.3333>,<at 1 0 RCIM_Cubic>]}
       RedCurve {keys [<at 0.0553 40 RCIM_Cubic>,<at 0.4938 5>,<at 1 0 RCIM_Cubic>]}
      }})
    Scale Alpha=&FN_FrontEndFloatModifier(
     Gameplay=&FNTimeOfDayScalar(Night Scale=0.5, Morning Scale=11, Day Scale=10, Evening Scale=0.5)
     Front End=1
    )
    Scale Color=&VectorFromFloat(Value=&Lerp_Float(A=20, B=40, Alpha=$System.FN_Integrator001.NormalizedValue))
   )
  ]
  rendererClasses [&NiagaraSpriteRendererProperties]
  renderers [{rendererClass=&NiagaraSpriteRendererProperties}]
 }
 GPU_Sparks001 {
  emitterUpdateScript [
   &EmitterState(Life Cycle Mode=::NewEnumerator1, Loop Behavior=::NewEnumerator1, Loop Delay=0.03, Loop Duration=0.15, UseLoopDelay=true)
   &SpawnRate(
    SpawnRate=&Multiply_Float(
     A=1100
     B=&FloatFromCurve(
      FloatCurve=DataInterface() {propertyValues {Curve {keys [<at 0 1 RCIM_Cubic>, {arriveTangent=-13.385, interpMode=RCIM_Cubic, leaveTangent=-13.385, tangentMode=RCTM_User, time=0.1, value=0}]}}}
      CurveIndex=$Emitter.Age
     )
    )
    Use Spawn Probability=false
   )
  ]
  particleSpawnScript [
   &InitializeParticle(
    Color=<rgba 40 9.8591 0 1>
    Lifetime Max=0.4
    Lifetime Min=0.1
    Lifetime Mode=::NewEnumerator1
    Mass Max=0.05
    Mass Min=0.02
    Mass Mode=::NewEnumerator3
    Mesh Scale Mode=::NewEnumerator3
    Mesh Uniform Scale=5
    Sprite Size Max=<3 7>
    Sprite Size Min=<2 3>
    Sprite Size Mode=::NewEnumerator2
    UsePositionOffset=false
    Write Material Random=true
   )
   &SphereLocation(
    Coordinate Space:ENiagaraCoordinateSpace=::Simulation
    MaskSpawnGroup=false
    Module.Override Randomness=false
    Offset=&Lerp_Vector(A=<0 0 0>, B=<0 0 0>, Alpha=&MakeFloatFromInt(Int=$Particles.SpawnGroup))
    Sphere Radius=2
    Surface Only=false
   ) {enabled=false}
   &ShapeLocation(
    MaskSpawnGroup=false
    Offset=<0 0 10>
    Offset Coordinate Space=::Local
    Offset Mode=::NewEnumerator0
    Override Seed=false
    Sphere Radius=2
    UseOwnerScale=true
   )
   &AddVelocity(
    Cone Angle=30
    Cone Axis=<0 0 1>
    Module.UseConeFalloff=false
    Module.UseVelDistribution=false
    Override Seed=false
    Velocity Mode=::NewEnumerator2
    Velocity Speed=&RandomRangeFloat(Minimum=100, Maximum=450)
   )
  ]
  particleUpdateScript [
   &ParticleState()
   &LinearForce(Force=&VectorFromCurve(
     VectorCurve=DataInterface() {propertyValues {XCurve {keys [<at 0 1>,<at 0.1 0>]}, YCurve {keys [<at 0 1>,<at 0.1 0>]}, ZCurve {keys [<at 0 1>,<at 0.1 0>]}}}
     Scale Curve=&FN_FrontEndVector3Modifier(Gameplay=<0 2000 2000>, Front End=<0 0 2500>)
    )) {enabled=false}
   &CurlNoiseForce(
    Curl Noise Cone Mask Angle=90
    Curl Noise Cone Mask Falloff Angle=90
    Mask Curl Noise=true
    Noise Frequency=200
    Noise Strength=100
    Pan Noise=false
   ) {enabled=false}
   &Drag(Drag=4, Use Linear Drag=true, Use Rotational Drag=false)
   &PointForce(Module.Use Falloff=false, Module.UseLinearForce=false) {enabled=false}
   &GravityForce(Gravity=<0 0 -300>)
   &SolveForcesAndVelocity(Clamp Velocity=false, Limit Acceleration=false)
   &ScaleSpriteSizeBySpeed(
    Max Scale Factor=<1 1>
    Min Scale Factor=<0.5 0.5>
    Sample Scale Factor By Curve=true
    Scale Factor Curve=DataInterface() {propertyValues {Curve {keys [
        {arriveTangent=2.865, interpMode=RCIM_Cubic, time=0.1247, value=-0.0006}
        {arriveTangent=4.1522, interpMode=RCIM_Cubic, leaveTangent=4.1522, tangentMode=RCTM_User, time=0.3948, value=0.4893}
        <at 0.9958 0.9974 RCIM_Cubic>
       ]}}}
    Velocity Threshold=500
   )
   &ScaleSpriteSize(Scale Factor=&Vector2DFromCurve(
     Vector2Curve=DataInterface() {propertyValues {
       XCurve {keys [
         <at 0 0 RCIM_Cubic>
         {arriveTangent=-1.6411, interpMode=RCIM_Cubic, leaveTangent=-1.6411, tangentMode=RCTM_User, time=0.2457, value=1.0046}
         <at 1 0>
        ]}
       YCurve {keys [<at 0 0 RCIM_Cubic>,<at 0.2457 1.0046 RCIM_Cubic>,<at 1 0>]}
      }}
     ScaleCurve=<1 1>
    ))
   &Color(
    Color=&ColorFromCurve(ColorCurve=DataInterface() {propertyValues {
       AlphaCurve {keys [<at 0 0>,<at 0.0353 1 RCIM_Cubic>,<at 1 0 RCIM_Cubic>]}
       BlueCurve {keys [<at 0.0553 0.7793 RCIM_Cubic>,<at 0.4938 0>,<at 1 0 RCIM_Cubic>]}
       GreenCurve {keys [<at 0.0553 0 RCIM_Cubic>,<at 0.4938 0.3333>,<at 1 0 RCIM_Cubic>]}
       RedCurve {keys [<at 0.0553 40 RCIM_Cubic>,<at 0.4938 5>,<at 1 0 RCIM_Cubic>]}
      }})
    Scale Alpha=&FN_FrontEndFloatModifier(
     Gameplay=&FNTimeOfDayScalar(Night Scale=0.5, Morning Scale=11, Day Scale=10, Evening Scale=0.5)
     Front End=1
    )
    Scale Color=&VectorFromFloat(Value=&Lerp_Float(A=20, B=40, Alpha=$System.FN_Integrator001.NormalizedValue))
   )
  ]
  rendererClasses [&NiagaraSpriteRendererProperties]
  renderers [{rendererClass=&NiagaraSpriteRendererProperties}]
 }
 GPU_Sparks003 {
  emitterUpdateScript [
   &EmitterState(Life Cycle Mode=::NewEnumerator1, Loop Behavior=::NewEnumerator1, Loop Duration=0.2, UseLoopDelay=false)
   &SpawnBurst_Instantaneous(Spawn Count=20, Use Spawn Probability=false, UseLoopCountLimit=false)
   &SpawnRate(
    SpawnRate=&Multiply_Float(
     A=100
     B=&FloatFromCurve(
      FloatCurve=DataInterface() {propertyValues {Curve {keys [<at 0 1 RCIM_Cubic>, {arriveTangent=-13.385, interpMode=RCIM_Cubic, leaveTangent=-13.385, tangentMode=RCTM_User, time=0.1, value=0}]}}}
      CurveIndex=$Emitter.Age
     )
    )
    Use Spawn Probability=false
   ) {enabled=false}
  ]
  particleSpawnScript [
   &InitializeParticle(
    Lifetime Max=0.15
    Lifetime Min=0.05
    Lifetime Mode=::NewEnumerator1
    Mass Max=0.05
    Mass Min=0.02
    Mass Mode=::NewEnumerator3
    Sprite Rotation Mode=::NewEnumerator2
    Sprite Size=&Multiply_Vector2D_ByFloat(Float=&RandomRangeFloat(Minimum=15, Maximum=30), Vector2D=<0.5 1>)
    Sprite Size Mode=::NewEnumerator4
    UsePositionOffset=false
    Write Material Random=true
   )
   &SphereLocation(
    Coordinate Space:ENiagaraCoordinateSpace=::Simulation
    MaskSpawnGroup=false
    Module.Override Randomness=false
    Offset=&Lerp_Vector(A=<0 0 0>, B=<0 0 0>, Alpha=&MakeFloatFromInt(Int=$Particles.SpawnGroup))
    Sphere Radius=2
    Surface Only=false
   ) {enabled=false}
   &AddVelocity(
    Module.Use Falloff Distance=false
    Override Seed=false
    Velocity Mode=::NewEnumerator1
    Velocity Speed=&RandomRangeFloat(Minimum=50, Maximum=200)
   )
   &ShapeLocation(
    MaskSpawnGroup=false
    Offset=<0 0 10>
    Offset Coordinate Space=::Local
    Offset Mode=::NewEnumerator0
    Override Seed=false
    Sphere Radius=2
    UseOwnerScale=true
   )
   &SubUVAnimation(
    "Sprite Renderer "=DataInterface() {propertyValues {SpriteRenderer=&self:GPU_Sparks003_4.NiagaraSpriteRendererProperties_0}}
    SubUV Animation Mode:ENiagara_SubUVLookupModeV2=::NewEnumerator2
    UseEndFrame=false
    UseStartFrame=false
   )
   &CameraOffset(Camera Offset Amount=15)
  ]
  particleUpdateScript [
   &ParticleState()
   &GravityForce(Gravity=<0 0 -200>)
   &ScaleSpriteSize(Uniform Curve Sprite Scale=DataInterface() {propertyValues {Curve {keys [<at 0 0 RCIM_Cubic>,<at 0.8 1 RCIM_Cubic>]}}})
   &Drag(Drag=5, Use Linear Drag=true, Use Rotational Drag=false)
   &SolveForcesAndVelocity(Clamp Velocity=false, Limit Acceleration=false)
   &ScaleSpriteSizeBySpeed(Max Scale Factor=<0.5 2>, Velocity Threshold=600)
   &ScaleColor(
    Scale Alpha=&FloatFromCurve(FloatCurve=DataInterface() {propertyValues {Curve {keys [<at 0 1 RCIM_Cubic>, {arriveTangent=-0.186, interpMode=RCIM_Cubic, leaveTangent=-0.186, tangentMode=RCTM_User, time=1, value=0}]}}})
    Scale RGB=&MakeVectorFromLinearColorRGB(Color=&ColorFromCurve(ColorCurve=DataInterface() {propertyValues {
        AlphaCurve {keys []}
        BlueCurve {keys [<at 0.0999 9.5833 RCIM_Cubic>,<at 0.8248 0.2775 RCIM_Cubic>]}
        GreenCurve {keys [<at 0.0999 1.5 RCIM_Cubic>,<at 0.8248 0.03 RCIM_Cubic>]}
        RedCurve {keys [<at 0.0999 50 RCIM_Cubic>,<at 0.8248 3 RCIM_Cubic>]}
       }}))
    ScaleA=true
    ScaleRGB=true
   )
  ]
  rendererClasses [&NiagaraSpriteRendererProperties]
  renderers [{rendererClass=&NiagaraSpriteRendererProperties}]
 }
 Smoke {
  bEnabled=false
  emitterUpdateScript [
   &EmitterState(Life Cycle Mode=::NewEnumerator1, Loop Behavior=::NewEnumerator1, Loop Duration=1, UseLoopDelay=false)
   &SpawnBurst_Instantaneous(Spawn Count=40, Use Spawn Probability=false, UseLoopCountLimit=false)
  ]
  particleSpawnScript [
   &InitializeParticle(
    Color Maximum=<rgba 0.076 0.1 0.1 0.8>
    Color Minimum=<rgba 0.048 0.055 0.06 0.4>
    Color Mode=::NewEnumerator2
    Lifetime Max=0.7
    Lifetime Min=0.2
    Lifetime Mode=::NewEnumerator1
    Sprite Rotation Mode=::NewEnumerator2
    Sprite Size Mode=::NewEnumerator1
    Uniform Sprite Size Max=50
    Uniform Sprite Size Min=30
    UsePositionOffset=false
    Write Material Random=true
   )
   &PartitionParticles()
   &ShapeLocation(
    MaskSpawnGroup=false
    Offset=<0 0 7>
    Offset Coordinate Space=::Local
    Offset Mode=::NewEnumerator0
    Override Seed=false
    Sphere Radius=8
    UseOwnerScale=true
   )
   &AddVelocity(Module.Use Falloff Distance=false, Override Seed=false, Velocity Mode=::NewEnumerator1, Velocity Speed=60)
  ]
  particleUpdateScript [
   &GravityForce(Gravity=<0 0 250>) {enabled=false}
   &ParticleState()
   &SpriteRotationRate(Rotation Rate=80)
   &ScaleSpriteSize(Uniform Curve Sprite Scale=DataInterface() {propertyValues {Curve {keys [{arriveTangent=0.1274, interpMode=RCIM_Cubic, leaveTangent=0.1274, tangentMode=RCTM_User, time=0, value=0.4}, <at 0.8 1 RCIM_Cubic>]}}})
   &ScaleColor(
    Scale Alpha=&FloatFromCurve(FloatCurve=DataInterface() {propertyValues {Curve {keys [<at 0 0 RCIM_Cubic>,<at 0.1 1 RCIM_Cubic>,<at 1 0 RCIM_Cubic>]}}}, Scale Curve=0.3)
    Scale RGB=&MakeVectorFromLinearColorRGB(Color=&ColorFromCurve(ColorCurve=DataInterface() {propertyValues {
        AlphaCurve {keys []}
        BlueCurve {keys [<at 0.1738 0.7 RCIM_Cubic>,<at 0.7949 1 RCIM_Cubic>]}
        GreenCurve {keys [<at 0.1738 0.7105 RCIM_Cubic>,<at 0.7949 1 RCIM_Cubic>]}
        RedCurve {keys [<at 0.1738 7 RCIM_Cubic>,<at 0.7949 1 RCIM_Cubic>]}
       }}))
    ScaleA=true
    ScaleRGB=true
   )
   &DynamicMaterialParameters(
    Index 0 Param 1=&Add_Float(
     A=&RandomRangeFloat(Minimum=0, Maximum=1)
     B=&FloatFromCurve(
      FloatCurve=DataInterface() {propertyValues {Curve {keys [
          <at 0 0 RCIM_Cubic>
          {arriveTangent=1.5947, interpMode=RCIM_Cubic, leaveTangent=1.5947, tangentMode=RCTM_User, time=0.2557, value=0.5257}
          <at 1 1 RCIM_Cubic>
         ]}}}
      CurveIndex=$Particles.Age
      Scale Curve=0.1
     )
    )
    Module.Param0WriteEnabled=true
    Module.Param1WriteEnabled=true
    Module.Param2WriteEnabled=true
    Module.Param3WriteEnabled=true
   ) {module=DynamicMaterialParameters001}
   &Drag(Drag=4, Use Linear Drag=true, Use Rotational Drag=true)
   &SolveForcesAndVelocity(Clamp Velocity=false, Limit Acceleration=false)
   &ScaleSpriteSizeBySpeed(
    Max Scale Factor=&Vector2DFromFloat(Value=0.8)
    Min Scale Factor=&Vector2DFromFloat(Value=1)
   ) {enabled=false}
   &CameraOffset(Camera Offset Amount=8)
  ]
  rendererClasses [&NiagaraSpriteRendererProperties]
  renderers [{rendererClass=&NiagaraSpriteRendererProperties}]
 }
 SmokeTrail {
  emitterUpdateScript [
   &EmitterState(Life Cycle Mode=::NewEnumerator1, Loop Delay=8, Loop Duration=3, UseLoopDelay=false)
   &SpawnRate(SpawnRate=50, Use Spawn Probability=false) {enabled=false, module=SpawnRate001}
   &SpawnRate(
    SpawnRate=&FloatFromCurve(
     FloatCurve=DataInterface() {propertyValues {Curve {keys [
         {arriveTangent=0.0107, interpMode=RCIM_Cubic, leaveTangent=0.0107, tangentMode=RCTM_User, time=9, value=0}
         {arriveTangent=-0.0015, interpMode=RCIM_Cubic, leaveTangent=-0.0015, tangentMode=RCTM_User, time=9.5, value=1}
         {arriveTangent=-0.0241, interpMode=RCIM_Cubic, leaveTangent=-0.0241, tangentMode=RCTM_User, time=10, value=1}
         {arriveTangent=0.0226, interpMode=RCIM_Cubic, leaveTangent=0.0226, tangentMode=RCTM_User, time=11.2, value=0}
        ]}}}
     CurveIndex=$Engine.System.Age
     Scale Curve=60
    )
    Use Spawn Probability=false
   )
  ]
  particleSpawnScript [
   &InitializeParticle(
    Color Mode=::NewEnumerator0
    Lifetime=0.5
    Position Offset=<0 0 10>
    Position Offset Coordinate Space=::Local
    Ribbon Width=4
    Ribbon Width Mode=::NewEnumerator1
    Sprite Size Mode=::NewEnumerator3
    UsePositionOffset=true
    Write Material Random=true
   )
   &InheritVelocity(Speed Limit=true, UseSpeedThreshold=false) {enabled=false}
   &AddVelocity(Velocity=&MakeVector(
     X=0
     Y=0
     Z=&FloatFromCurve(
      FloatCurve=DataInterface() {propertyValues {Curve {keys [<at 8 0 RCIM_Cubic>,<at 10.5 1 RCIM_Cubic>,<at 11.5 0 RCIM_Cubic>]}}}
      CurveIndex=$System.Age
      Scale Curve=250
     )
    ))
   &InheritVelocity(Inherited Velocity Speed Limit=50, Speed Limit=true, UseSpeedThreshold=false) {module=InheritVelocity001}
   &CameraOffset(Camera Offset Amount=-5)
  ]
  particleUpdateScript [
   &ParticleState()
   /SetVariables(Particles.RibbonWidth=&Multiply_Float(
     A=&FloatFromCurve(FloatCurve=DataInterface() {propertyValues {Curve {keys [<at 0 0.2 RCIM_Cubic>,<at 0.7 1 RCIM_Cubic>]}}}, Scale Curve=10)
     B=&FloatFromCurve(
      FloatCurve=DataInterface() {propertyValues {Curve {keys [
          {arriveTangent=0.0107, interpMode=RCIM_Cubic, leaveTangent=0.0107, tangentMode=RCTM_User, time=9, value=0}
          {arriveTangent=-0.0015, interpMode=RCIM_Cubic, leaveTangent=-0.0015, tangentMode=RCTM_User, time=9.5, value=1}
          {arriveTangent=-0.0241, interpMode=RCIM_Cubic, leaveTangent=-0.0241, tangentMode=RCTM_User, time=10, value=1}
          {arriveTangent=0.0226, interpMode=RCIM_Cubic, leaveTangent=0.0226, tangentMode=RCTM_User, time=11.2, value=0}
         ]}}}
      CurveIndex=$System.Age
     )
    )) {isSetParameters=true}
   &GravityForce(Gravity=<0 0 200>)
   &CurlNoiseForce(
    Noise Frequency=80
    Noise Strength=&FloatFromCurve(FloatCurve=DataInterface() {propertyValues {Curve {keys [<at 0 0.5 RCIM_Cubic>,<at 0.5 1 RCIM_Cubic>]}}}, Scale Curve=1000)
    Pan Noise=false
   )
   &Drag(Drag=10, Use Linear Drag=true, Use Rotational Drag=true)
   &SolveForcesAndVelocity(Clamp Velocity=false, Limit Acceleration=false)
   /SetVariables(Particles.RibbonLinkOrder:NiagaraFloat=&OneMinusFloat(Float=$Particles.NormalizedAge)) {isSetParameters=true}
   &ScaleColor(
    Scale Alpha=&FloatFromCurve(FloatCurve=DataInterface() {propertyValues {Curve {keys [<at 0.3 1 RCIM_Cubic>,<at 1 0 RCIM_Cubic>]}}}, Scale Curve=0.3)
    Scale RGB=&MakeVectorFromLinearColorRGB(Color=&ColorFromCurve(ColorCurve=DataInterface() {propertyValues {
        AlphaCurve {keys []}
        BlueCurve {keys [<at 0.0099 0.32 RCIM_Cubic>, {arriveTangent=-0.4008, interpMode=RCIM_Cubic, time=0.4042, value=0.16}, <at 0.8635 0.024>]}
        GreenCurve {keys [<at 0.0099 0.3 RCIM_Cubic>, {arriveTangent=-0.3737, interpMode=RCIM_Cubic, time=0.4042, value=0.16}, <at 0.8635 0.024>]}
        RedCurve {keys [<at 0.0099 1.5 RCIM_Cubic>, {arriveTangent=-1.9905, interpMode=RCIM_Cubic, time=0.4042, value=0.2}, <at 0.8635 0.03>]}
       }}))
    ScaleA=true
    ScaleRGB=true
   )
   &ScaleColor(
    Scale Alpha=&FloatFromCurve(FloatCurve=DataInterface() {propertyValues {Curve {keys [
         {arriveTangent=0.0107, interpMode=RCIM_Cubic, leaveTangent=0.0107, tangentMode=RCTM_User, time=9, value=0}
         {arriveTangent=-0.0015, interpMode=RCIM_Cubic, leaveTangent=-0.0015, tangentMode=RCTM_User, time=10, value=1}
         {arriveTangent=-0.0241, interpMode=RCIM_Cubic, leaveTangent=-0.0241, tangentMode=RCTM_User, time=10.6, value=1}
         {arriveTangent=-1.4085, interpMode=RCIM_Cubic, leaveTangent=-1.4085, tangentMode=RCTM_User, time=11.9, value=0}
        ]}}})
    ScaleA=true
    ScaleRGB=false
   ) {enabled=false, module=ScaleColor001}
  ]
  rendererClasses [&NiagaraRibbonRendererProperties]
  renderers [{rendererClass=&NiagaraRibbonRendererProperties}]
 }
 Smoke_Big {
  emitterUpdateScript [
   &EmitterState()
   &SpawnRate(
    SpawnRate=&Multiply_Float(
     A=&FloatFromCurve(
      FloatCurve=DataInterface() {propertyValues {Curve {keys [{arriveTangent=0.1729, interpMode=RCIM_Cubic, leaveTangent=0.1729, tangentMode=RCTM_User, time=0.3, value=0}, <at 0.8 1 RCIM_Cubic>]}}}
      CurveIndex=$Emitter.Age
      Scale Curve=400
     )
     B=$System.FlareDuration
    )
    Use Spawn Probability=false
   ) {enabled=false}
   &SpawnRate(
    SpawnRate=&Multiply_Float(
     A=&FloatFromCurve(
      FloatCurve=DataInterface() {propertyValues {Curve {keys [{arriveTangent=0.1729, interpMode=RCIM_Cubic, leaveTangent=0.1729, tangentMode=RCTM_User, time=0.3, value=0}, <at 0.8 1 RCIM_Cubic>]}}}
      CurveIndex=$Emitter.Age
      Scale Curve=&Multiply_Float(A=300, B=$System.WaveAnimationIntensity)
     )
     B=$System.FlareDuration
    )
    Use Spawn Probability=false
   ) {module=SpawnRate001}
  ]
  particleSpawnScript [
   &InitializeParticle(
    Color Maximum=<rgba 0.076 0.1 0.1 0.8>
    Color Minimum=<rgba 0.048 0.055 0.06 0.4>
    Color Mode=::NewEnumerator2
    Lifetime Max=0.8
    Lifetime Min=0.6
    Lifetime Mode=::NewEnumerator1
    Mass Max=1
    Mass Min=0.8
    Mass Mode=::NewEnumerator3
    Sprite Rotation Mode=::NewEnumerator2
    Sprite Size Mode=::NewEnumerator1
    Uniform Sprite Size Max=40
    Uniform Sprite Size Min=25
    UsePositionOffset=false
    Write Material Random=true
   )
   &ShapeLocation(
    MaskSpawnGroup=false
    Offset=<0 0 7>
    Offset Coordinate Space=::Local
    Offset Mode=::NewEnumerator0
    Override Seed=false
    Sphere Radius=8
    UseOwnerScale=true
   )
   &InheritVelocity(Inherited Velocity Speed Limit=50, Speed Limit=true, UseSpeedThreshold=false)
   &AddVelocity(Velocity=<0 0 260>, Velocity Speed Scale=&RandomRangeFloat(Minimum=0.75, Maximum=1)) {module=AddVelocity001}
  ]
  particleUpdateScript [
   &GravityForce(Gravity=<0 0 100>)
   &ParticleState()
   &CurlNoiseForce(
    Noise Frequency=20
    Noise Strength=&Multiply_Float(
     A=$System.SpeedFlare
     B=&Multiply_Float(A=0.4, B=&OneMinusFloat(Float=$Particles.NormalizedAge))
    )
    Pan Noise=false
   )
   &SpriteRotationRate(Rotation Rate=80)
   &ScaleSpriteSize(Uniform Curve Sprite Scale=DataInterface() {propertyValues {Curve {keys [{arriveTangent=0.1274, interpMode=RCIM_Cubic, leaveTangent=0.1274, tangentMode=RCTM_User, time=0, value=0.5}, <at 0.8 1 RCIM_Cubic>]}}})
   &ScaleColor(
    Scale Alpha=&FloatFromCurve(
     FloatCurve=DataInterface() {propertyValues {Curve {keys [<at 0 0 RCIM_Cubic>,<at 0.13 1 RCIM_Cubic>,<at 1 0 RCIM_Cubic>]}}}
     Scale Curve=&FN_Select_Float_from_Time_Of_Day(Morning=0.35, Day=0.3, Evening=0.35, Night=0.45)
    )
    Scale RGB=&MakeVectorFromLinearColorRGB(Color=&ColorFromCurve(ColorCurve=DataInterface() {propertyValues {
        AlphaCurve {keys []}
        BlueCurve {keys [<at 0.0978 0.6 RCIM_Cubic>,<at 0.2994 1.12>,<at 0.8732 3.2 RCIM_Cubic>]}
        GreenCurve {keys [<at 0.0978 0.699 RCIM_Cubic>,<at 0.2994 1.2347>,<at 0.8732 3.2667 RCIM_Cubic>]}
        RedCurve {keys [<at 0.0978 60 RCIM_Cubic>,<at 0.2994 8>,<at 0.8732 4 RCIM_Cubic>]}
       }}))
    ScaleA=true
    ScaleRGB=true
   )
   &ScaleColor(Scale Alpha=$System.FlareFlameDuration, ScaleA=true, ScaleRGB=false) {enabled=false, module=ScaleColor001}
   &DynamicMaterialParameters(
    Index 0 Param 1=&Add_Float(
     A=&RandomRangeFloat(Minimum=0, Maximum=1)
     B=&FloatFromCurve(
      FloatCurve=DataInterface() {propertyValues {Curve {keys [
          <at 0 0 RCIM_Cubic>
          {arriveTangent=1.5947, interpMode=RCIM_Cubic, leaveTangent=1.5947, tangentMode=RCTM_User, time=0.2557, value=0.5257}
          <at 1 1 RCIM_Cubic>
         ]}}}
      CurveIndex=$Particles.Age
      Scale Curve=0.1
     )
    )
    Module.Param0WriteEnabled=true
    Module.Param1WriteEnabled=true
    Module.Param2WriteEnabled=true
    Module.Param3WriteEnabled=true
   ) {module=DynamicMaterialParameters001}
   &Drag(Drag=5, Use Linear Drag=true, Use Rotational Drag=true)
   &SolveForcesAndVelocity(Clamp Velocity=false, Limit Acceleration=false)
   &ScaleSpriteSizeBySpeed(
    Max Scale Factor=&Vector2DFromFloat(Value=0.8)
    Min Scale Factor=&Vector2DFromFloat(Value=1)
   ) {enabled=false}
   &CameraOffset(Camera Offset Amount=5)
  ]
  rendererClasses [&NiagaraSpriteRendererProperties]
  renderers [
   {rendererClass=&NiagaraSpriteRendererProperties}
   {rendererClass=&NiagaraSpriteRendererProperties, rendererIndex=1}
  ]
 }
 Smoke_High_Pressure {
  emitterUpdateScript [
   &EmitterState()
   &SpawnRate(
    SpawnRate=&Multiply_Float(
     A=&FloatFromCurve(
      FloatCurve=DataInterface() {propertyValues {Curve {keys [{arriveTangent=0.1729, interpMode=RCIM_Cubic, leaveTangent=0.1729, tangentMode=RCTM_User, time=0.1, value=0}, <at 0.8 1 RCIM_Cubic>]}}}
      CurveIndex=$Emitter.Age
      Scale Curve=70
     )
     B=$System.FlareFlameDuration
    )
    Use Spawn Probability=false
   )
  ]
  particleSpawnScript [
   &InitializeParticle(
    Color Maximum=<rgba 0.076 0.1 0.1 0.8>
    Color Minimum=<rgba 0.048 0.055 0.06 0.4>
    Color Mode=::NewEnumerator2
    Lifetime Max=0.2
    Lifetime Min=0.15
    Lifetime Mode=::NewEnumerator1
    Sprite Rotation Mode=::NewEnumerator2
    Sprite Size Mode=::NewEnumerator1
    Uniform Sprite Size Max=20
    Uniform Sprite Size Min=10
    UsePositionOffset=false
    Write Material Random=true
   )
   &PartitionParticles() {enabled=false}
   &ShapeLocation(
    MaskSpawnGroup=false
    Offset=<0 0 7>
    Offset Coordinate Space=::Local
    Offset Mode=::NewEnumerator0
    Override Seed=false
    Sphere Radius=4
    UseOwnerScale=true
   )
   &AddVelocity(
    Module.Use Falloff Distance=false
    Origin Offset=<0 0 -20>
    Origin Offset Coordinate Space=::Simulation
    Override Seed=false
    Velocity Mode=::NewEnumerator1
    Velocity Speed=300
   ) {enabled=false}
   &InheritVelocity(Inherited Velocity Speed Limit=150, Speed Limit=true, UseSpeedThreshold=false)
   &AddVelocity(Velocity=<0 0 300>, Velocity Speed Scale=&RandomRangeFloat(Minimum=0.75, Maximum=1)) {module=AddVelocity001}
  ]
  particleUpdateScript [
   &GravityForce(Gravity=<0 0 50>)
   &ParticleState()
   &SpriteRotationRate(Rotation Rate=120)
   &ScaleSpriteSize(Uniform Curve Sprite Scale=DataInterface() {propertyValues {Curve {keys [{arriveTangent=0.1274, interpMode=RCIM_Cubic, leaveTangent=0.1274, tangentMode=RCTM_User, time=0, value=0.6}, <at 0.8 1 RCIM_Cubic>]}}})
   &ScaleColor(
    Scale Alpha=&Multiply_Float(
     A=&FloatFromCurve(
      FloatCurve=DataInterface() {propertyValues {Curve {keys [
          <at 0 0 RCIM_Cubic>
          {arriveTangent=0.0728, interpMode=RCIM_Cubic, leaveTangent=0.0728, tangentMode=RCTM_User, time=0.1, value=1}
          <at 1 0 RCIM_Cubic>
         ]}}}
      Scale Curve=0.7
     )
     B=$System.FlareDuration
    )
    Scale RGB=&MakeVectorFromLinearColorRGB(Color=&ColorFromCurve(ColorCurve=DataInterface() {propertyValues {
        AlphaCurve {keys []}
        BlueCurve {keys [<at 0.1739 2.565>, {interpMode=RCIM_Cubic, leaveTangent=-2.3787, time=0.7267, value=0.9533}]}
        GreenCurve {keys [<at 0.1739 2.1>, {interpMode=RCIM_Cubic, leaveTangent=-1.8078, time=0.7267, value=0.8}]}
        RedCurve {keys [<at 0.1739 30>, {interpMode=RCIM_Cubic, leaveTangent=-35.1059, time=0.7267, value=10}]}
       }}))
    ScaleA=true
    ScaleRGB=true
   )
   &ScaleColor(Scale Alpha=$System.FlareFlameDuration, ScaleA=true, ScaleRGB=false) {enabled=false, module=ScaleColor001}
   &DynamicMaterialParameters(
    Index 0 Param 1=&Add_Float(
     A=&RandomRangeFloat(Minimum=0, Maximum=1)
     B=&FloatFromCurve(
      FloatCurve=DataInterface() {propertyValues {Curve {keys [
          <at 0 0 RCIM_Cubic>
          {arriveTangent=1.5947, interpMode=RCIM_Cubic, leaveTangent=1.5947, tangentMode=RCTM_User, time=0.2557, value=0.5257}
          <at 1 1 RCIM_Cubic>
         ]}}}
      CurveIndex=$Particles.Age
      Scale Curve=0.1
     )
    )
    Module.Param0WriteEnabled=true
    Module.Param1WriteEnabled=true
    Module.Param2WriteEnabled=true
    Module.Param3WriteEnabled=true
   ) {module=DynamicMaterialParameters001}
   &Drag(Drag=6, Use Linear Drag=true, Use Rotational Drag=true)
   &SolveForcesAndVelocity(Clamp Velocity=false, Limit Acceleration=false)
   &ScaleSpriteSizeBySpeed(
    Max Scale Factor=&Vector2DFromFloat(Value=0.8)
    Min Scale Factor=&Vector2DFromFloat(Value=1)
   ) {enabled=false}
   &CameraOffset(Camera Offset Amount=8)
  ]
  rendererClasses [&NiagaraSpriteRendererProperties]
  renderers [
   {rendererClass=&NiagaraSpriteRendererProperties}
   {rendererClass=&NiagaraSpriteRendererProperties, rendererIndex=1}
  ]
 }
}
summary {
 systemName=NS_Emote_CrispRover
 emitters [
  {emitterName=GPU_Sparks, simTarget=GPUComputeSim, rendererClasses [&NiagaraSpriteRendererProperties]}
  {
   emitterName=Flames
   rendererClasses [
    &NiagaraSpriteRendererProperties
    &NiagaraLightRendererProperties
    &NiagaraMeshRendererProperties
   ]
  }
  {emitterName=Smoke_Big, rendererClasses [&NiagaraSpriteRendererProperties]}
  {bEnabled=false, emitterName=Empty001, rendererClasses [&NiagaraSpriteRendererProperties]}
  {emitterName=GPU_Sparks001, rendererClasses [&NiagaraSpriteRendererProperties]}
  {emitterName=GPU_Sparks003, rendererClasses [&NiagaraSpriteRendererProperties]}
  {bEnabled=false, emitterName=Smoke, rendererClasses [&NiagaraSpriteRendererProperties]}
  {emitterName=Smoke_High_Pressure, rendererClasses [&NiagaraSpriteRendererProperties]}
  {emitterName=SmokeTrail, rendererClasses [&NiagaraRibbonRendererProperties]}
 ]
 userVariables [
  User.bIsFrontEnd {type=NiagaraBool}
  User.bIsFrontEndPreview {type=NiagaraBool}
  User.NormalizedNotifyProgress {type=NiagaraFloat, defaultValue=0, description=""}
  User.SkeletalMesh {
   type=NiagaraDataInterfaceSkeletalMesh
   defaultValue {
    struct=NiagaraExt_VariableValue_DataInterface
    value {dataInterface=&self:NiagaraDataInterfaceSkeletalMesh_0}
   }
  }
 ]
}
